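# Looping mono MetaSound for Beast Girl's movement vocals, repeatedly playing random walk VO clips.
# After a short start delay, a repeat trigger fires at intervals mapped from movement speed (faster is shorter), scaled by LFO noise.
# A trigger gate, fed back through deferred variables, closes on each trigger and reopens when playback finishes, preventing overlap.
# Each clip gets a random pitch shift, and character affiliation selects the interpolated local, enemy or non-hostile volume.
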MetaSoundSource /BeastGirlStormClawsGameplay/Assets/Audio/SoundCues/Vox/MSS_BeastGirl_MovementVocals
inputs [
 Character.Affiliation {type=Enum:FortAudioAffiliation}
 Character.Health {type=Float}
 Character.UniqueID {type=Int32}
 Movement.Speed {type=Float}
 Movement.SpeedAzimuth {type=Float}
 Movement.SpeedElevation {type=Float}
 Sounds {type=WaveAsset:Array}
 UE.Source.OnPlay {type=Trigger}
 Volume.Enemy {type=Float}
 Volume.Local {type=Float}
 Volume.Non-hostile {type=Float}
]
interfaces [Fort.Character:1.0, Fort.Movement:1.0, UE.OutputFormat.Mono:1.0, UE.Source:1.1]
outputs ["UE.OutputFormat.Mono.Audio:0" {type=Audio}]
pages [Default {
  input_defaults {
   Character.Affiliation=0
   Character.Health=0
   Character.UniqueID=0
   Movement.Speed=1000
   Movement.SpeedAzimuth=0
   Movement.SpeedElevation=0
   UE.Source.OnPlay=null
   Volume.Enemy=1
   Volume.Local=1
   Volume.Non-hostile=0.4
   Sounds [
    /BeastGirlStormClawsGameplay/Assets/Audio/Wavs/VO/BeastGirlStorm_Walk_VO_01
    /BeastGirlStormClawsGameplay/Assets/Audio/Wavs/VO/BeastGirlStorm_Walk_VO_02
    /BeastGirlStormClawsGameplay/Assets/Audio/Wavs/VO/BeastGirlStorm_Walk_VO_03
    /BeastGirlStormClawsGameplay/Assets/Audio/Wavs/VO/BeastGirlStorm_Walk_VO_04
    /BeastGirlStormClawsGameplay/Assets/Audio/Wavs/VO/BeastGirlStorm_Walk_VO_05
    /BeastGirlStormClawsGameplay/Assets/Audio/Wavs/VO/BeastGirlStorm_Walk_VO_06
    /BeastGirlStormClawsGameplay/Assets/Audio/Wavs/VO/BeastGirlStorm_Walk_VO_07
    /BeastGirlStormClawsGameplay/Assets/Audio/Wavs/VO/BeastGirlStorm_Walk_VO_08
    /BeastGirlStormClawsGameplay/Assets/Audio/Wavs/VO/BeastGirlStorm_Walk_VO_09
    /BeastGirlStormClawsGameplay/Assets/Audio/Wavs/VO/BeastGirlStorm_Walk_VO_10
    /BeastGirlStormClawsGameplay/Assets/Audio/Wavs/VO/BeastGirlStorm_Walk_VO_11
    /BeastGirlStormClawsGameplay/Assets/Audio/Wavs/VO/BeastGirlStorm_Walk_VO_12
    /BeastGirlStormClawsGameplay/Assets/Audio/Wavs/VO/BeastGirlStorm_Walk_VO_13
    /BeastGirlStormClawsGameplay/Assets/Audio/Wavs/VO/BeastGirlStorm_Walk_VO_14
    /BeastGirlStormClawsGameplay/Assets/Audio/Wavs/VO/BeastGirlStorm_Walk_VO_15
    /BeastGirlStormClawsGameplay/Assets/Audio/Wavs/VO/BeastGirlStorm_Walk_VO_16
   ]
  }
  nodes [
   CrossfadeTriggerRouteFloat4_0 = "Crossfade.Trigger Route (Float, 4):Float"(
    In 2=^Volume.Enemy
    In 0=^Volume.Local
    In 1=^Volume.Non-hostile
    In 3=^Volume.Non-hostile
    Crossfade Value=Convert.Float:Int32(Int32=Convert.Int32:Enum:FortAudioAffiliation("Enum:FortAudioAffiliation"=^Character.Affiliation).Int32).Float
   )
   LfoFrequencyNoise_0 = UE.Lfo Frequency Noise:Audio() {also {defaults {Max Value=1.5, Min Value=0.75, Rate Jitter=1}}}
   n1 = UE.Multiply:Float(
    AdditionalOperands=@LfoFrequencyNoise_0.Out
    PrimaryOperand=MapRange.MapRange:Float(In=UE.InterpTo:Audio(Target=^Movement.Speed).Value) {also {defaults {In Range A=200, In Range B=1000, Out Range A=1.5, Out Range B=0.8}}}.Out Value
   )
   TriggerDelay_0 = UE.Trigger Delay(In=^UE.Source.OnPlay) {also {defaults {Delay Time=0.5}}}
   n2 = VariableDeferredAccessor.Trigger(Variable=InitVariable.Trigger().Variable)
   n3 = VariableDeferredAccessor.Trigger(Variable=InitVariable.Trigger().Variable)
   TriggerControl_0 = UE.Trigger Control(
    Trigger In=UE.TriggerRepeat(Period=UE.ConversionFloatToTime(In=@n1.Out).Out, Start=@TriggerDelay_0.Out).RepeatOut
    Open=@n2.Value
    Close=@n3.Value
   )
   n4 = VariableAccessor.Trigger(Variable=VariableMutator.Trigger(Value=@TriggerControl_0.Trigger Out, Variable=@n3.Variable).Variable)
   ArrayRandomGet_0 = Array.Random Get:WaveAsset:Array(In Array=^Sounds, No Repeats=UE.Divide:Int32(PrimaryOperand=Array.Num:WaveAsset:Array(Array=^Sounds).Num).Out, Next=@n4.Value)
   WavePlayer_0 = UE.Wave Player:Mono(
    Play=@ArrayRandomGet_0.On Next
    Wave Asset=@ArrayRandomGet_0.Value
    Pitch Shift=UE.RandomFloat(Next=VariableAccessor.Trigger(Variable=@n4.Variable).Value) {also {defaults {Max=2, Min=-1}}}.Value
   )
   Multiply_0 = UE.Multiply:Audio by Float(AdditionalOperands=UE.InterpTo:Audio(Target=@CrossfadeTriggerRouteFloat4_0.Out).Value, PrimaryOperand=@WavePlayer_0.Out Mono)
   VariableMutator.Trigger(Variable=@n2.Variable, Value=@WavePlayer_0.On Finished)
  ]
  links [{dst="@UE.OutputFormat.Mono.Audio:0", src=Multiply_0.Out}]
 }]
properties {
 assetImportData=&self:AssetImportData
 attenuationSettings=/BeastGirlStormClawsGameplay/Assets/Audio/ATT_BeastGirlStorm_VO_Idle
 bLooping=true
 duration=10000
 maxDistance=4954.375
 priority=0
 sampleRate=48000
 soundClassObject=/Game/Sounds/FortSoundClassesAndModes/Fort_Creatures_Large_Critical_PSM
 volume=2
 concurrencySet [
  /BeastGirlStormClawsGameplay/Assets/Audio/CONC_BeastGirlStorm_VO_Local_Prevent_New
  /BeastGirlStormClawsGameplay/Assets/Audio/CONC_BeastGirlStorm_VO_Global
 ]
}
