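# Sound cue for the activation voice line of the BeastGirl Storm Claws gameplay, currently empty with no root node or sound nodes.
# It carries only playback setup: a boosted volume, the Fort_Creatures sound class and a dedicated VO attenuation asset.
# Two concurrency assets govern simultaneous playback, a global one and a local one named to prevent new instances.
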SoundCue /BeastGirlStormClawsGameplay/Assets/Audio/SoundCues/Vox/SC_BeastGirlStormClaws_Activate_VO
root=null
nodes []
properties {
 attenuationSettings=/BeastGirlStormClawsGameplay/Assets/Audio/ATT_BeastGirlStorm_VO
 maxDistance=7927
 soundClassObject=/Game/Sounds/FortSoundClassesAndModes/Fort_Creatures
 volumeMultiplier=2
 attenuationOverrides {bAttenuate=false, lPFRadiusMin=17, stereoSpread=1000}
 concurrencySet [
  /BeastGirlStormClawsGameplay/Assets/Audio/CONC_BeastGirlStorm_VO_Local_Prevent_New
  /BeastGirlStormClawsGameplay/Assets/Audio/CONC_BeastGirlStorm_VO_Global
 ]
}
