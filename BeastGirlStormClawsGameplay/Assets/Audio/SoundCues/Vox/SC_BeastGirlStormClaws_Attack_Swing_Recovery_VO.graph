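# Sound cue for BeastGirlStormClaws voice-over during attack swing recovery.
# A Random node picks among eight recorded variations and six empty slots at equal weight, so some triggers play silence.
# A root Modulator varies pitch at or below normal to add variety.
# It uses the BeastGirlStorm VO attenuation, a critical large-creature sound class, and both local prevent-new and global concurrency limits.
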SoundCue /BeastGirlStormClawsGameplay/Assets/Audio/SoundCues/Vox/SC_BeastGirlStormClaws_Attack_Swing_Recovery_VO
root=Modulator_0
nodes [
 {id=Modulator_0, children [Random_2], properties {pitchMax=1, pitchMin=0.85, volumeMax=1}}
 {
  id=Random_2
  children [
   WavePlayer_14
   WavePlayer_16
   WavePlayer_17
   WavePlayer_18
   WavePlayer_19
   WavePlayer_20
   WavePlayer_21
   WavePlayer_22
   null
   null
   null
   null
   null
   null
  ]
  properties {weights [1,1,1,1,1,1,1,1,1,1,1,1,1,1]}
 }
 {
  id=WavePlayer_14
  properties {soundWaveAssetPtr=/BeastGirlStormClawsGameplay/Assets/Audio/Wavs/VO/BeastGirlStormClaws_Attack_Swing_Recovery_01}
 }
 {
  id=WavePlayer_16
  properties {soundWaveAssetPtr=/BeastGirlStormClawsGameplay/Assets/Audio/Wavs/VO/BeastGirlStormClaws_Attack_Swing_Recovery_03}
 }
 {
  id=WavePlayer_17
  properties {soundWaveAssetPtr=/BeastGirlStormClawsGameplay/Assets/Audio/Wavs/VO/BeastGirlStormClaws_Attack_Swing_Recovery_04}
 }
 {
  id=WavePlayer_18
  properties {soundWaveAssetPtr=/BeastGirlStormClawsGameplay/Assets/Audio/Wavs/VO/BeastGirlStormClaws_Attack_Swing_Recovery_05}
 }
 {
  id=WavePlayer_19
  properties {soundWaveAssetPtr=/BeastGirlStormClawsGameplay/Assets/Audio/Wavs/VO/BeastGirlStormClaws_Attack_Swing_Recovery_06}
 }
 {
  id=WavePlayer_20
  properties {soundWaveAssetPtr=/BeastGirlStormClawsGameplay/Assets/Audio/Wavs/VO/BeastGirlStormClaws_Attack_Swing_Recovery_07}
 }
 {
  id=WavePlayer_21
  properties {soundWaveAssetPtr=/BeastGirlStormClawsGameplay/Assets/Audio/Wavs/VO/BeastGirlStormClaws_Attack_Swing_Recovery_08}
 }
 {
  id=WavePlayer_22
  properties {soundWaveAssetPtr=/BeastGirlStormClawsGameplay/Assets/Audio/Wavs/VO/BeastGirlStormClaws_Attack_Swing_Recovery_09}
 }
]
properties {
 attenuationSettings=/BeastGirlStormClawsGameplay/Assets/Audio/ATT_BeastGirlStorm_VO
 maxDistance=9908.75
 soundClassObject=/Game/Sounds/FortSoundClassesAndModes/Fort_Creatures_Large_Critical_PSM
 volumeMultiplier=0.65
 attenuationOverrides {bAttenuate=false, lPFRadiusMin=17, stereoSpread=1000}
 concurrencySet [
  /BeastGirlStormClawsGameplay/Assets/Audio/CONC_BeastGirlStorm_VO_Local_Prevent_New
  /BeastGirlStormClawsGameplay/Assets/Audio/CONC_BeastGirlStorm_VO_Global
 ]
}
