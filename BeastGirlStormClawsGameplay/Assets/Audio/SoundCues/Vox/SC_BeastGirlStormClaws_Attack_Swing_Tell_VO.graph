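# Sound cue for the BeastGirl StormClaws vocal tell on a swing attack.
# A Random node picks among five VO waves and four empty slots at equal weight, so some triggers stay silent.
# Playback uses the BeastGirlStorm VO attenuation asset and the large critical creature sound class, at slightly reduced volume.
# Local prevent-new and global concurrency sets limit overlapping vocals.
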SoundCue /BeastGirlStormClawsGameplay/Assets/Audio/SoundCues/Vox/SC_BeastGirlStormClaws_Attack_Swing_Tell_VO
root=Random_0
nodes [
 {
  id=Random_0
  children [WavePlayer_0,WavePlayer_1,WavePlayer_2,WavePlayer_3,WavePlayer_4,null,null,null,null]
  properties {weights [1,1,1,1,1,1,1,1,1]}
 }
 {id=WavePlayer_0, properties {soundWaveAssetPtr=/BeastGirlStormClawsGameplay/Assets/Audio/Wavs/VO/BeastGirlStormClaws_Attack_Swing_Tell_01}}
 {id=WavePlayer_1, properties {soundWaveAssetPtr=/BeastGirlStormClawsGameplay/Assets/Audio/Wavs/VO/BeastGirlStormClaws_Attack_Swing_Tell_02}}
 {id=WavePlayer_2, properties {soundWaveAssetPtr=/BeastGirlStormClawsGameplay/Assets/Audio/Wavs/VO/BeastGirlStormClaws_Attack_Swing_Tell_03}}
 {id=WavePlayer_3, properties {soundWaveAssetPtr=/BeastGirlStormClawsGameplay/Assets/Audio/Wavs/VO/BeastGirlStormClaws_Attack_Swing_Tell_04}}
 {id=WavePlayer_4, properties {soundWaveAssetPtr=/BeastGirlStormClawsGameplay/Assets/Audio/Wavs/VO/BeastGirlStormClaws_Attack_Swing_Tell_05}}
]
properties {
 attenuationSettings=/BeastGirlStormClawsGameplay/Assets/Audio/ATT_BeastGirlStorm_VO
 maxDistance=9908.75
 soundClassObject=/Game/Sounds/FortSoundClassesAndModes/Fort_Creatures_Large_Critical_PSM
 volumeMultiplier=0.844
 attenuationOverrides {bAttenuate=false, lPFRadiusMin=17, stereoSpread=1000}
 concurrencySet [
  /BeastGirlStormClawsGameplay/Assets/Audio/CONC_BeastGirlStorm_VO_Local_Prevent_New
  /BeastGirlStormClawsGameplay/Assets/Audio/CONC_BeastGirlStorm_VO_Global
 ]
}
