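# Sound cue for the BeastGirl StormClaws creature's vocal tell on its swing attack.
# A Random node picks one of five equally weighted VO wave variations.
# Output is volume-boosted, routed to the large critical creature sound class, and uses the character's shared VO attenuation asset.
# Two concurrency settings, a local prevent-new and a global one, limit overlapping lines.
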SoundCue /BeastGirlStormClawsGameplay/Assets/Audio/SoundCues/Vox/SC_BeastGirlStormClaws_Attack_Tell_Swing_VO
root=Random_1
nodes [
 {id=Random_1, children [WavePlayer_8, WavePlayer_9, WavePlayer_10, WavePlayer_11, WavePlayer_12], properties {weights [1,1,1,1,1]}}
 {
  id=WavePlayer_10
  properties {soundWaveAssetPtr=/BeastGirlStormClawsGameplay/Assets/Audio/Wavs/VO/BeastGirlStormClaws_Attack_Swing_Tell_VO_03}
 }
 {
  id=WavePlayer_11
  properties {soundWaveAssetPtr=/BeastGirlStormClawsGameplay/Assets/Audio/Wavs/VO/BeastGirlStormClaws_Attack_Swing_Tell_VO_04}
 }
 {
  id=WavePlayer_12
  properties {soundWaveAssetPtr=/BeastGirlStormClawsGameplay/Assets/Audio/Wavs/VO/BeastGirlStormClaws_Attack_Swing_Tell_VO_05}
 }
 {
  id=WavePlayer_8
  properties {soundWaveAssetPtr=/BeastGirlStormClawsGameplay/Assets/Audio/Wavs/VO/BeastGirlStormClaws_Attack_Swing_Tell_VO_01}
 }
 {
  id=WavePlayer_9
  properties {soundWaveAssetPtr=/BeastGirlStormClawsGameplay/Assets/Audio/Wavs/VO/BeastGirlStormClaws_Attack_Swing_Tell_VO_02}
 }
]
properties {
 attenuationSettings=/BeastGirlStormClawsGameplay/Assets/Audio/ATT_BeastGirlStorm_VO
 maxDistance=9908.75
 soundClassObject=/Game/Sounds/FortSoundClassesAndModes/Fort_Creatures_Large_Critical_PSM
 volumeMultiplier=1.5
 attenuationOverrides {bAttenuate=false, lPFRadiusMin=17, stereoSpread=1000}
 concurrencySet [
  /BeastGirlStormClawsGameplay/Assets/Audio/CONC_BeastGirlStorm_VO_Local_Prevent_New
  /BeastGirlStormClawsGameplay/Assets/Audio/CONC_BeastGirlStorm_VO_Global
 ]
}
