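# Sound cue for the BeastGirl StormClaws uppercut attack voice-over.
# A random node picks among five takes and four equally weighted empty slots, so some triggers can be silent, then feeds a shared random pitch modulator.
# Branches on islocal and isenemy set loudness: local or unset is full, non-local enemies are reduced, non-enemies lowest.
# Character VO attenuation and local and global concurrency sets govern playback.
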SoundCue /BeastGirlStormClawsGameplay/Assets/Audio/SoundCues/Vox/SC_BeastGirlStormClaws_Attack_Uppercut_VO
root=Branch_0
nodes [
 {id=Branch_0, children [Modulator_2,Branch_3,Modulator_2], properties {boolParameterName=islocal}}
 {id=Branch_3, children [Modulator_4,Modulator_3,Modulator_4], properties {boolParameterName=isenemy}}
 {id=Modulator_0, children [Random_1], properties {pitchMax=1.1, pitchMin=0.85, volumeMax=1, volumeMin=1}}
 {id=Modulator_2, children [Modulator_0], properties {pitchMax=1, pitchMin=1, volumeMax=1, volumeMin=1}}
 {id=Modulator_3, children [Modulator_0], properties {pitchMax=1, pitchMin=1, volumeMax=0.5, volumeMin=0.5}}
 {id=Modulator_4, children [Modulator_0], properties {pitchMax=1, pitchMin=1, volumeMax=0.75, volumeMin=0.75}}
 {
  id=Random_1
  children [WavePlayer_3,WavePlayer_4,WavePlayer_5,WavePlayer_6,WavePlayer_7,null,null,null,null]
  properties {weights [1,1,1,1,1,1,1,1,1]}
 }
 {
  id=WavePlayer_3
  properties {soundWaveAssetPtr=/BeastGirlStormClawsGameplay/Assets/Audio/Wavs/VO/BeastGirlStormClaws_Attack_Swing_UpperCut_VO_01}
 }
 {
  id=WavePlayer_4
  properties {soundWaveAssetPtr=/BeastGirlStormClawsGameplay/Assets/Audio/Wavs/VO/BeastGirlStormClaws_Attack_Swing_UpperCut_VO_02}
 }
 {
  id=WavePlayer_5
  properties {soundWaveAssetPtr=/BeastGirlStormClawsGameplay/Assets/Audio/Wavs/VO/BeastGirlStormClaws_Attack_Swing_UpperCut_VO_03}
 }
 {
  id=WavePlayer_6
  properties {soundWaveAssetPtr=/BeastGirlStormClawsGameplay/Assets/Audio/Wavs/VO/BeastGirlStormClaws_Attack_Swing_UpperCut_VO_04}
 }
 {
  id=WavePlayer_7
  properties {soundWaveAssetPtr=/BeastGirlStormClawsGameplay/Assets/Audio/Wavs/VO/BeastGirlStormClaws_Attack_Swing_UpperCut_VO_05}
 }
]
properties {
 attenuationSettings=/BeastGirlStormClawsGameplay/Assets/Audio/ATT_BeastGirlStorm_VO
 maxDistance=9908.75
 soundClassObject=/Game/Sounds/FortSoundClassesAndModes/Fort_Creatures_Large_Critical_PSM
 volumeMultiplier=1
 attenuationOverrides {bAttenuate=false, lPFRadiusMin=17, stereoSpread=1000}
 concurrencySet [
  /BeastGirlStormClawsGameplay/Assets/Audio/CONC_BeastGirlStorm_VO_Local_Prevent_New
  /BeastGirlStormClawsGameplay/Assets/Audio/CONC_BeastGirlStorm_VO_Global
 ]
}
