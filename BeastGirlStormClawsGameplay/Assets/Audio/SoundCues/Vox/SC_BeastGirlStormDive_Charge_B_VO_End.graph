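# Sound cue for the Beast Girl Storm Dive "Charge B" ending vocal.
# Nested islocal and isenemy branches set volume: full for local or unset, half for non-local enemies or unset, quarter for non-enemies.
# All branches share one mixer layering close, medium and distant variants, each randomized from three waves with its own attenuation.
# Medium adds pitch variation; medium and distant override the sound class, and two concurrency sets limit playback.
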SoundCue /BeastGirlStormClawsGameplay/Assets/Audio/SoundCues/Vox/SC_BeastGirlStormDive_Charge_B_VO_End
root=Branch_0
nodes [
 {
  id=Attenuation_118
  children [Random_0]
  properties {attenuationSettings=/BeastGirlStormClawsGameplay/Assets/Audio/ATT_BeastGirlStormClaws_Close, attenuationOverrides {stereoSpread=29871.3984}}
 }
 {
  id=Attenuation_119
  children [Random_1]
  properties {attenuationSettings=/BeastGirlStormClawsGameplay/Assets/Audio/ATT_BeastGirlStormClaws_Med}
 }
 {
  id=Attenuation_3
  children [Random_2]
  properties {
   attenuationSettings=/BeastGirlStormClawsGameplay/Assets/Audio/ATT_BeastGirlStormClaws_Dist
   attenuationOverrides {bAttenuateWithLPF=true, falloffDistance=2000, lPFRadiusMax=5000, lPFRadiusMin=2000, attenuationShapeExtents {x=3000}}
  }
 }
 {id=Branch_0, children [Modulator_2,Branch_3,Modulator_2], properties {boolParameterName=islocal}}
 {id=Branch_3, children [Modulator_4,Modulator_3,Modulator_4], properties {boolParameterName=isenemy}}
 {id=Mixer_1, children [Attenuation_118, Modulator_0, SoundClass_2], properties {inputVolume [1,0.8,1]}}
 {id=Modulator_0, children [SoundClass_76], properties {pitchMin=0.8, volumeMax=1, volumeMin=1}}
 {id=Modulator_2, children [Mixer_1], properties {pitchMax=1, pitchMin=1, volumeMax=1, volumeMin=1}}
 {id=Modulator_3, children [Mixer_1], properties {pitchMax=1, pitchMin=1, volumeMax=0.25, volumeMin=0.25}}
 {id=Modulator_4, children [Mixer_1], properties {pitchMax=1, pitchMin=1, volumeMax=0.5, volumeMin=0.5}}
 {id=Random_0, children [WavePlayer_2,WavePlayer_3,WavePlayer_4], properties {weights [1,1,1]}}
 {id=Random_1, children [WavePlayer_12, WavePlayer_13, WavePlayer_14], properties {weights [1,1,1]}}
 {id=Random_2, children [WavePlayer_9, WavePlayer_10, WavePlayer_11], properties {weights [1,1,1]}}
 {
  id=SoundClass_2
  children [Attenuation_3]
  properties {soundClassOverride=&Fort_WeaponRanged_Fire_3P_PSM}
 }
 {
  id=SoundClass_76
  children [Attenuation_119]
  properties {soundClassOverride=&Fort_WeaponRanged_Fire_3P_PSM}
 }
 {
  id=WavePlayer_10
  properties {soundWaveAssetPtr=/BeastGirlStormClawsGameplay/Assets/Audio/Wavs/VO/BeastGirlStormDive_Charge_B_VO_End_Distant_02}
 }
 {
  id=WavePlayer_11
  properties {soundWaveAssetPtr=/BeastGirlStormClawsGameplay/Assets/Audio/Wavs/VO/BeastGirlStormDive_Charge_B_VO_End_Distant_03}
 }
 {
  id=WavePlayer_12
  properties {soundWaveAssetPtr=/BeastGirlStormClawsGameplay/Assets/Audio/Wavs/VO/BeastGirlStormDive_Charge_B_VO_End_Medium_01}
 }
 {
  id=WavePlayer_13
  properties {soundWaveAssetPtr=/BeastGirlStormClawsGameplay/Assets/Audio/Wavs/VO/BeastGirlStormDive_Charge_B_VO_End_Medium_02}
 }
 {
  id=WavePlayer_14
  properties {soundWaveAssetPtr=/BeastGirlStormClawsGameplay/Assets/Audio/Wavs/VO/BeastGirlStormDive_Charge_B_VO_End_Medium_03}
 }
 {id=WavePlayer_2, properties {soundWaveAssetPtr=/BeastGirlStormClawsGameplay/Assets/Audio/Wavs/VO/BeastGirlStormDive_Charge_B_VO_End_01}}
 {id=WavePlayer_3, properties {soundWaveAssetPtr=/BeastGirlStormClawsGameplay/Assets/Audio/Wavs/VO/BeastGirlStormDive_Charge_B_VO_End_02}}
 {id=WavePlayer_4, properties {soundWaveAssetPtr=/BeastGirlStormClawsGameplay/Assets/Audio/Wavs/VO/BeastGirlStormDive_Charge_B_VO_End_03}}
 {
  id=WavePlayer_9
  properties {soundWaveAssetPtr=/BeastGirlStormClawsGameplay/Assets/Audio/Wavs/VO/BeastGirlStormDive_Charge_B_VO_End_Distant_01}
 }
]
properties {
 attenuationSettings=/BeastGirlStormClawsGameplay/Assets/Audio/ATT_BeastGirlStorm_VO_Distant
 maxDistance=30000
 pitchMultiplier=1.1
 priority=3
 soundClassObject=/Game/Sounds/FortSoundClassesAndModes/Fort_Creatures_Large_Critical_PSM
 volumeMultiplier=1.4
 attenuationOverrides {bAttenuate=false, lPFRadiusMin=17, stereoSpread=1000}
 busSends [{audioBus=null, minSendLevel=0, sendLevel=1, soundSourceBus=null}]
 concurrencySet [
  /BeastGirlStormClawsGameplay/Assets/Audio/CONC_BeastGirlStorm_VO_Local_Prevent_New
  /BeastGirlStormClawsGameplay/Assets/Audio/CONC_BeastGirlStorm_VO_Global
 ]
}
