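# Sound cue for the Beast Girl Storm Dive voice line, layered by distance.
# A mixer blends close, medium and distant layers, each picking one of three random waves under its own attenuation, with pitch variation on the medium layer.
# Branches on islocal and isenemy feed that mixer, playing non-local voices at half volume whether enemy or not.
# Medium and distant layers override the sound class, and two concurrency sets limit playback.
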SoundCue /BeastGirlStormClawsGameplay/Assets/Audio/SoundCues/Vox/SC_BeastGirlStormDive_Dive_VO
root=Branch_0
nodes [
 {
  id=Attenuation_118
  children [Random_0]
  properties {attenuationSettings=/BeastGirlStormClawsGameplay/Assets/Audio/ATT_BeastGirlStormClaws_Close, attenuationOverrides {stereoSpread=29871.3984}}
 }
 {
  id=Attenuation_119
  children [Random_1]
  properties {attenuationSettings=/BeastGirlStormClawsGameplay/Assets/Audio/ATT_BeastGirlStormClaws_Med}
 }
 {
  id=Attenuation_3
  children [Random_2]
  properties {
   attenuationSettings=/BeastGirlStormClawsGameplay/Assets/Audio/ATT_BeastGirlStormClaws_Dist
   attenuationOverrides {bAttenuateWithLPF=true, falloffDistance=2000, lPFRadiusMax=5000, lPFRadiusMin=2000, attenuationShapeExtents {x=3000}}
  }
 }
 {id=Branch_0, children [Modulator_0,Branch_3,Modulator_0], properties {boolParameterName=islocal}}
 {id=Branch_3, children [Modulator_5,Modulator_1,Modulator_5], properties {boolParameterName=isenemy}}
 {id=Mixer_1, children [Attenuation_118, Modulator_2, SoundClass_2], properties {inputVolume [1,1,1]}}
 {id=Modulator_0, children [Mixer_1], properties {pitchMax=1, pitchMin=1, volumeMax=1, volumeMin=1}}
 {id=Modulator_1, children [Mixer_1], properties {pitchMax=1, pitchMin=1, volumeMax=0.5, volumeMin=0.5}}
 {id=Modulator_2, children [SoundClass_76], properties {pitchMin=0.8, volumeMax=1, volumeMin=1}}
 {id=Modulator_5, children [Mixer_1], properties {pitchMax=1, pitchMin=1, volumeMax=0.5, volumeMin=0.5}}
 {id=Random_0, children [WavePlayer_0,WavePlayer_1,WavePlayer_2], properties {weights [1,1,1]}}
 {id=Random_1, children [WavePlayer_6,WavePlayer_7,WavePlayer_8], properties {weights [1,1,1]}}
 {id=Random_2, children [WavePlayer_3,WavePlayer_4,WavePlayer_5], properties {weights [1,1,1]}}
 {
  id=SoundClass_2
  children [Attenuation_3]
  properties {soundClassOverride=&Fort_WeaponRanged_Fire_3P_PSM}
 }
 {
  id=SoundClass_76
  children [Attenuation_119]
  properties {soundClassOverride=&Fort_WeaponRanged_Fire_3P_PSM}
 }
 {id=WavePlayer_0, properties {soundWaveAssetPtr=/BeastGirlStormClawsGameplay/Assets/Audio/Wavs/VO/BeastGirlStormDive_Dive_VO_02}}
 {id=WavePlayer_1, properties {soundWaveAssetPtr=/BeastGirlStormClawsGameplay/Assets/Audio/Wavs/VO/BeastGirlStormDive_Dive_VO_03}}
 {id=WavePlayer_2, properties {soundWaveAssetPtr=/BeastGirlStormClawsGameplay/Assets/Audio/Wavs/VO/BeastGirlStormDive_Dive_VO_01}}
 {id=WavePlayer_3, properties {soundWaveAssetPtr=/BeastGirlStormClawsGameplay/Assets/Audio/Wavs/VO/BeastGirlStormDive_Dive_VO_Distant_01}}
 {id=WavePlayer_4, properties {soundWaveAssetPtr=/BeastGirlStormClawsGameplay/Assets/Audio/Wavs/VO/BeastGirlStormDive_Dive_VO_Distant_02}}
 {id=WavePlayer_5, properties {soundWaveAssetPtr=/BeastGirlStormClawsGameplay/Assets/Audio/Wavs/VO/BeastGirlStormDive_Dive_VO_Distant_03}}
 {id=WavePlayer_6, properties {soundWaveAssetPtr=/BeastGirlStormClawsGameplay/Assets/Audio/Wavs/VO/BeastGirlStormDive_Dive_VO_Medium_01}}
 {id=WavePlayer_7, properties {soundWaveAssetPtr=/BeastGirlStormClawsGameplay/Assets/Audio/Wavs/VO/BeastGirlStormDive_Dive_VO_Medium_02}}
 {id=WavePlayer_8, properties {soundWaveAssetPtr=/BeastGirlStormClawsGameplay/Assets/Audio/Wavs/VO/BeastGirlStormDive_Dive_VO_Medium_03}}
]
properties {
 attenuationSettings=/BeastGirlStormClawsGameplay/Assets/Audio/ATT_BeastGirlStorm_VO_Distant
 maxDistance=30000
 priority=5
 soundClassObject=/Game/Sounds/FortSoundClassesAndModes/Fort_Creatures_Large_Critical_PSM
 volumeMultiplier=1.4
 attenuationOverrides {bAttenuate=false, lPFRadiusMin=17, stereoSpread=1000}
 concurrencySet [
  /BeastGirlStormClawsGameplay/Assets/Audio/CONC_BeastGirlStorm_VO_Local_Prevent_New
  /BeastGirlStormClawsGameplay/Assets/Audio/CONC_BeastGirlStorm_VO_Global
 ]
}
