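# SoundCue for Beast Girl's Storm Dive grunt voice-over in the Storm Claws gameplay content.
# A Random node picks between two grunt wave variations with equal weight.
# It plays through the melee weapon sound class with a boosted volume and the shared Beast Girl Storm VO attenuation asset.
# Two concurrency sets, one local prevent-new and one global, limit overlapping playback.
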SoundCue /BeastGirlStormClawsGameplay/Assets/Audio/SoundCues/Vox/SC_BeastGirlStormDive_Grunt_VO
root=Random_0
nodes [
 {id=Random_0, children [WavePlayer_3,WavePlayer_4], properties {weights [1,1]}}
 {id=WavePlayer_3, properties {soundWaveAssetPtr=/BeastGirlStormClawsGameplay/Assets/Audio/Wavs/VO/BeastGirlStormDive_Grunt_VO_01}}
 {id=WavePlayer_4, properties {soundWaveAssetPtr=/BeastGirlStormClawsGameplay/Assets/Audio/Wavs/VO/BeastGirlStormDive_Grunt_VO_02}}
]
properties {
 attenuationSettings=/BeastGirlStormClawsGameplay/Assets/Audio/ATT_BeastGirlStorm_VO
 maxDistance=9908.75
 soundClassObject=&Fort_WeaponMelee_PSM
 volumeMultiplier=1.783
 attenuationOverrides {bAttenuate=false, lPFRadiusMin=17, stereoSpread=1000}
 concurrencySet [
  /BeastGirlStormClawsGameplay/Assets/Audio/CONC_BeastGirlStorm_VO_Local_Prevent_New
  /BeastGirlStormClawsGameplay/Assets/Audio/CONC_BeastGirlStorm_VO_Global
 ]
}
