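# Sound cue for BeastGirlStorm idle creature vocalizations.
# A quality-level root sends its first two levels to an equal-weight random pick of fifteen voice clips, and the last two to a four-clip subset of them.
# Playback uses the creature vocalization sound class, a dedicated idle VO attenuation asset, and both local prevent-new and global concurrency sets.
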SoundCue /BeastGirlStormClawsGameplay/Assets/Audio/SoundCues/Vox/SC_BeastGirlStorm_Idle_VO
root=QualityLevel_0
nodes [
 {id=QualityLevel_0, children [Random_1,Random_1,Random_0,Random_0]}
 {id=Random_0, children [WavePlayer_6,WavePlayer_5,WavePlayer_4,WavePlayer_3], properties {weights [1,1,1,1]}}
 {
  id=Random_1
  children [
   WavePlayer_10
   WavePlayer_11
   WavePlayer_12
   WavePlayer_13
   WavePlayer_14
   WavePlayer_15
   WavePlayer_16
   WavePlayer_17
   WavePlayer_18
   WavePlayer_19
   WavePlayer_20
   WavePlayer_21
   WavePlayer_0
   WavePlayer_1
   WavePlayer_2
  ]
  properties {weights [1,1,1,1,1,1,1,1,1,1,1,1,1,1,1]}
 }
 {id=WavePlayer_0, properties {soundWaveAssetPtr=/BeastGirlStormClawsGameplay/Assets/Audio/Wavs/VO/BeastGirlStorm_Idle_VO_13}}
 {id=WavePlayer_1, properties {soundWaveAssetPtr=/BeastGirlStormClawsGameplay/Assets/Audio/Wavs/VO/BeastGirlStorm_Idle_VO_14}}
 {id=WavePlayer_10, properties {soundWaveAssetPtr=/BeastGirlStormClawsGameplay/Assets/Audio/Wavs/VO/BeastGirlStorm_Idle_VO_01}}
 {id=WavePlayer_11, properties {soundWaveAssetPtr=/BeastGirlStormClawsGameplay/Assets/Audio/Wavs/VO/BeastGirlStorm_Idle_VO_02}}
 {id=WavePlayer_12, properties {soundWaveAssetPtr=/BeastGirlStormClawsGameplay/Assets/Audio/Wavs/VO/BeastGirlStorm_Idle_VO_03}}
 {id=WavePlayer_13, properties {soundWaveAssetPtr=/BeastGirlStormClawsGameplay/Assets/Audio/Wavs/VO/BeastGirlStorm_Idle_VO_04}}
 {id=WavePlayer_14, properties {soundWaveAssetPtr=/BeastGirlStormClawsGameplay/Assets/Audio/Wavs/VO/BeastGirlStorm_Idle_VO_05}}
 {id=WavePlayer_15, properties {soundWaveAssetPtr=/BeastGirlStormClawsGameplay/Assets/Audio/Wavs/VO/BeastGirlStorm_Idle_VO_06}}
 {id=WavePlayer_16, properties {soundWaveAssetPtr=/BeastGirlStormClawsGameplay/Assets/Audio/Wavs/VO/BeastGirlStorm_Idle_VO_07}}
 {id=WavePlayer_17, properties {soundWaveAssetPtr=/BeastGirlStormClawsGameplay/Assets/Audio/Wavs/VO/BeastGirlStorm_Idle_VO_08}}
 {id=WavePlayer_18, properties {soundWaveAssetPtr=/BeastGirlStormClawsGameplay/Assets/Audio/Wavs/VO/BeastGirlStorm_Idle_VO_09}}
 {id=WavePlayer_19, properties {soundWaveAssetPtr=/BeastGirlStormClawsGameplay/Assets/Audio/Wavs/VO/BeastGirlStorm_Idle_VO_10}}
 {id=WavePlayer_2, properties {soundWaveAssetPtr=/BeastGirlStormClawsGameplay/Assets/Audio/Wavs/VO/BeastGirlStorm_Idle_VO_15}}
 {id=WavePlayer_20, properties {soundWaveAssetPtr=/BeastGirlStormClawsGameplay/Assets/Audio/Wavs/VO/BeastGirlStorm_Idle_VO_11}}
 {id=WavePlayer_21, properties {soundWaveAssetPtr=/BeastGirlStormClawsGameplay/Assets/Audio/Wavs/VO/BeastGirlStorm_Idle_VO_12}}
 {id=WavePlayer_3, properties {soundWaveAssetPtr=&BeastGirlStorm_Idle_VO_15}}
 {id=WavePlayer_4, properties {soundWaveAssetPtr=&BeastGirlStorm_Idle_VO_09}}
 {id=WavePlayer_5, properties {soundWaveAssetPtr=&BeastGirlStorm_Idle_VO_03}}
 {id=WavePlayer_6, properties {soundWaveAssetPtr=&BeastGirlStorm_Idle_VO_02}}
]
properties {
 attenuationSettings=/BeastGirlStormClawsGameplay/Assets/Audio/ATT_BeastGirlStorm_VO_Idle
 maxDistance=4954.375
 priority=0
 soundClassObject=/Game/Sounds/FortSoundClassesAndModes/Fort_Creatures_Vocalization_Normal
 attenuationOverrides {bAttenuate=false, lPFRadiusMin=17, stereoSpread=1000}
 concurrencySet [
  /BeastGirlStormClawsGameplay/Assets/Audio/CONC_BeastGirlStorm_VO_Local_Prevent_New
  /BeastGirlStormClawsGameplay/Assets/Audio/CONC_BeastGirlStorm_VO_Global
 ]
}
